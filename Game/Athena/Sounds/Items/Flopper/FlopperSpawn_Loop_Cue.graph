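# Looping ambient sound cue for the Flopper spawn, routed to the water ambient sound class.
# A Random node picks one of three equally weighted branches over the same looping wave, each with a short volume fade-in.
# Two branches add different start delays and long, differing pitch envelopes to vary instances.
# Overrides add distance low-pass, occlusion and listener-focus attenuation, and cap concurrency at three voices.
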SoundCue /Game/Athena/Sounds/Items/Flopper/FlopperSpawn_Loop_Cue
root=Random_0
nodes [
 {id=Delay_0, children [Enveloper_4], properties {delayMax=0.75, delayMin=0.75}}
 {id=Delay_1, children [Enveloper_5], properties {delayMax=2, delayMin=2}}
 {
  id=Enveloper_1
  children [WavePlayer_0]
  properties {pitchCurve {editorCurveData {keys [{time=0}]}}, volumeCurve {editorCurveData {keys [<at 0 0>,<at 0.2 1>]}}}
 }
 {
  id=Enveloper_2
  children [WavePlayer_0]
  properties {pitchCurve {editorCurveData {keys [{time=0}]}}, volumeCurve {editorCurveData {keys [<at 0 0>,<at 0.2 1>]}}}
 }
 {
  id=Enveloper_3
  children [WavePlayer_0]
  properties {pitchCurve {editorCurveData {keys [{time=0}]}}, volumeCurve {editorCurveData {keys [<at 0 0>,<at 0.2 1>]}}}
 }
 {
  id=Enveloper_4
  children [Enveloper_2]
  properties {
   loopEnd=20
   pitchCurve {editorCurveData {keys [{interpMode=RCIM_Cubic, time=0}, <at 10 0.98 RCIM_Cubic>, {time=20}]}}
   volumeCurve {editorCurveData {keys [<at 0 1>]}}
  }
 }
 {
  id=Enveloper_5
  children [Enveloper_3]
  properties {
   loopEnd=30
   pitchCurve {editorCurveData {keys [{interpMode=RCIM_Cubic, time=0}, <at 15 1.1 RCIM_Cubic>, {time=30}]}}
   volumeCurve {editorCurveData {keys [<at 0 1>]}}
  }
 }
 {id=Random_0, children [Delay_0,Enveloper_1,Delay_1], properties {weights [1,1,1]}}
 {id=WavePlayer_0, properties {bLooping=true, soundWaveAssetPtr=&/FlopperSpawn_Loop}}
]
properties {
 bOverrideAttenuation=true
 bOverrideConcurrency=true
 maxDistance=3790.1655
 soundClassObject=/Game/Sounds/FortSoundClassesAndModes/Fort_Water_Ambient
 volumeMultiplier=0.7
 attenuationOverrides {
  bAttenuateWithLPF=true
  bEnableListenerFocus=true
  bEnableOcclusion=true
  falloffDistance=2710.8213
  focusVolumeAttenuation=1.0775
  lPFFrequencyAtMax=200
  lPFRadiusMax=1281.8871
  lPFRadiusMin=442.5163
  nonFocusVolumeAttenuation=0.25
  occlusionInterpolationTime=2
  occlusionLowPassFilterFrequency=8000
  occlusionTraceChannel=ECC_GameTraceChannel11
  occlusionVolumeAttenuation=0.5
  stereoSpread=1707.1663
  attenuationShapeExtents {x=1079.3441}
 }
 concurrencyOverrides {maxCount=3, voiceStealReleaseTime=2}
}
