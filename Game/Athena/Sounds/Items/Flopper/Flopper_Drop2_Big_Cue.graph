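# Sound cue for dropping a big Flopper fish item in Athena.
# A Mixer layers two waves at equal volume, Small_Fish_Drop and Flopper_Drop2_Medium_01, each through its own Modulator at default settings for slight pitch and volume variation.
# The cue plays at reduced volume in the loot-collect sound class, using weapon-drop attenuation for distance falloff.
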SoundCue /Game/Athena/Sounds/Items/Flopper/Flopper_Drop2_Big_Cue
root=Mixer_0
nodes [
 {id=Mixer_0, children [Modulator_0,Modulator_1], properties {inputVolume [1,1]}}
 {id=Modulator_0, children [WavePlayer_3]}
 {id=Modulator_1, children [WavePlayer_0]}
 {id=WavePlayer_0, properties {soundWaveAssetPtr=&/Flopper_Drop2_Medium_01}}
 {id=WavePlayer_3, properties {soundWaveAssetPtr=&/Small_Fish_Drop}}
]
properties {
 attenuationSettings=&Fort_Weapon_Drop_Att
 maxDistance=3000
 soundClassObject=&Fort_Loot_Collect
 volumeMultiplier=0.5
}
