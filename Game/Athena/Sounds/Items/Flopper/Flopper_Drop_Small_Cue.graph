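# Sound cue for dropping a small Flopper fish item, using loot-collect sound class and weapon-drop attenuation.
# A mixer layers a random pick between two medium Flopper drop waves, at half input volume, with a Slurp Fish drop wave.
# A modulator fixes the Slurp Fish layer at a raised pitch, and the whole cue also plays pitched up and at reduced volume.
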SoundCue /Game/Athena/Sounds/Items/Flopper/Flopper_Drop_Small_Cue
root=Mixer_0
nodes [
 {id=Mixer_0, children [Random_1,Modulator_0], properties {inputVolume [0.5,1]}}
 {id=Modulator_0, children [WavePlayer_1], properties {pitchMax=1.5, pitchMin=1.5}}
 {id=Random_1, children [WavePlayer_0,WavePlayer_3], properties {weights [1,1]}}
 {id=WavePlayer_0, properties {soundWaveAssetPtr=&/Flopper_Drop2_Medium_01}}
 {id=WavePlayer_1, properties {soundWaveAssetPtr=&/Slurp_Fish_Drop}}
 {id=WavePlayer_3, properties {soundWaveAssetPtr=&/Flopper_Drop2_Medium_02}}
]
properties {
 attenuationSettings=&Fort_Weapon_Drop_Att
 maxDistance=3000
 pitchMultiplier=1.25
 soundClassObject=&Fort_Loot_Collect
 volumeMultiplier=0.45
}
