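# Sound cue for a Flopper item pickup, named as a high award variant, in the Loot Collect sound class.
# It plays a reused front-end reward-earned UI wave through a Modulator that fixes pitch at double, at reduced cue volume.
# Inline attenuation overrides keep it non-spatialized with a short distance falloff.
# Concurrency overrides allow one instance at a time, stopping the oldest.
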SoundCue /Game/Athena/Sounds/Items/Flopper/Flopper_Pickup2_AwardHigh_Cue
root=Modulator_0
nodes [
 {id=Modulator_0, children [WavePlayer_2], properties {pitchMax=2, pitchMin=2}}
 {id=WavePlayer_2, properties {soundWaveAssetPtr=/Game/Sounds/Fort_UI_Sounds/FrontEnd/New_FrontEnd/Fort_UI_RewardEarned_02}}
]
properties {
 attenuationSettings=&Fort_Weapon_Drop_Att
 bOverrideAttenuation=true
 bOverrideConcurrency=true
 maxDistance=1200
 soundClassObject=&Fort_Loot_Collect
 volumeMultiplier=0.5
 attenuationOverrides {bSpatialize=false, falloffDistance=600, attenuationShapeExtents {x=600}}
 concurrencyOverrides {maxCount=1, resolutionRule=StopOldest}
}
