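# Sound cue for a Flopper item pickup award, reusing the front-end select-row UI sound wave.
# A single wave player feeds a modulator at the root, with the cue's overall volume reduced.
# It belongs to the loot-collect sound class and borrows the weapon-drop attenuation settings for distance falloff.
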SoundCue /Game/Athena/Sounds/Items/Flopper/Flopper_Pickup2_Award_Cue
root=Modulator_0
nodes [
 {id=Modulator_0, children [WavePlayer_0]}
 {id=WavePlayer_0, properties {soundWaveAssetPtr=/Game/Sounds/Fort_UI_Sounds/FrontEnd/Fort_UI_CB_SelectRow_01}}
]
properties {
 attenuationSettings=&Fort_Weapon_Drop_Att
 maxDistance=3000
 soundClassObject=&Fort_Loot_Collect
 volumeMultiplier=0.6
}
