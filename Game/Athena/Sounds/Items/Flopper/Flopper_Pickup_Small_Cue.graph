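# Sound cue for picking up a small Flopper fish item in Athena, routed to the loot collect sound class with weapon pickup attenuation.
# A mixer layers three branches: a briefly delayed random choice between two medium flopper drop waves, a reused player melee equip sound, and a quieter, modulated thermal fish pickup wave.
# The whole cue plays at reduced volume and raised pitch.
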SoundCue /Game/Athena/Sounds/Items/Flopper/Flopper_Pickup_Small_Cue
root=Mixer_0
nodes [
 {id=Delay_0, children [Random_1], properties {delayMax=0.1, delayMin=0.1}}
 {id=Mixer_0, children [Delay_0,WavePlayer_1,Modulator_0], properties {inputVolume [1,1,0.6]}}
 {id=Modulator_0, children [WavePlayer_4]}
 {id=Random_1, children [WavePlayer_0,WavePlayer_3], properties {weights [1,1]}}
 {id=WavePlayer_0, properties {soundWaveAssetPtr=&/Flopper_Drop2_Medium_01}}
 {id=WavePlayer_1, properties {soundWaveAssetPtr=/Game/Sounds/Fort_GamePlay_Sounds/Player/player_melee_equip_02}}
 {id=WavePlayer_3, properties {soundWaveAssetPtr=&/Flopper_Drop2_Medium_02}}
 {id=WavePlayer_4, properties {soundWaveAssetPtr=&/Thremal_Fish_PU}}
]
properties {
 attenuationSettings=&Fort_Weapon_Pickup_Att
 maxDistance=3500
 pitchMultiplier=1.25
 soundClassObject=&Fort_Loot_Collect
 volumeMultiplier=0.45
}
