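# Sound cue for the slurp heard when a Jellyfish, a Flopper-family consumable item, is consumed.
# A single wave player feeds the Jellyfish_Consumed wave into a Modulator that adds very slight random pitch variation on each play.
# The cue plays in the consumables sound class with the Athena consumable-use attenuation and a slightly reduced volume.
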SoundCue /Game/Athena/Sounds/Items/Flopper/Jellyfish_Slurp_Consume_Cue
root=Modulator_0
nodes [
 {id=Modulator_0, children [WavePlayer_0], properties {pitchMax=1.01, pitchMin=0.99}}
 {id=WavePlayer_0, properties {soundWaveAssetPtr=&/Jellyfish_Consumed}}
]
properties {
 attenuationSettings=&Fort_Consumable_Use_Athena_Att
 maxDistance=2250
 soundClassObject=&Fort_Consumables
 volumeMultiplier=0.9
}
