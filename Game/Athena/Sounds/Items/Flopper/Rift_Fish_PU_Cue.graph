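# Sound cue for the Rift Fish Flopper item, apparently its pickup sound given the PU name and loot-collect sound class.
# A mixer layers the main pickup wave with a second element wave that starts after a short fixed delay and plays at lower volume.
# The cue uses weapon-drop attenuation and the loot-collect sound class, with reduced overall volume.
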SoundCue /Game/Athena/Sounds/Items/Flopper/Rift_Fish_PU_Cue
root=Mixer_0
nodes [
 {id=Delay_2, children [WavePlayer_3], properties {delayMax=0.12, delayMin=0.12}}
 {id=Mixer_0, children [WavePlayer_2,Delay_2], properties {inputVolume [1,0.65]}}
 {id=WavePlayer_2, properties {soundWaveAssetPtr=&/Rift_Fish_PU}}
 {id=WavePlayer_3, properties {soundWaveAssetPtr=&/Rift_Fish_PU_element_01}}
]
properties {
 attenuationSettings=&Fort_Weapon_Drop_Att
 maxDistance=3000
 soundClassObject=&Fort_Loot_Collect
 volumeMultiplier=0.65
}
