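# Sound cue for the fast rod pull-back sound of the Flopper item in Athena.
# A single wave player for Rod_Pull_Back_Fast feeds a Modulator node at the root, which is left at default settings.
# It plays through the player-interacts sound class and borrows the weapon-drop attenuation to fade with distance.
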SoundCue /Game/Athena/Sounds/Items/Flopper/Rod_Pull_Back_Fast_Cue
root=Modulator_1
nodes [
 {id=Modulator_1, children [WavePlayer_0]}
 {id=WavePlayer_0, properties {soundWaveAssetPtr=&/Rod_Pull_Back_Fast}}
]
properties {
 attenuationSettings=&Fort_Weapon_Drop_Att
 maxDistance=3000
 soundClassObject=/Game/Sounds/FortSoundClassesAndModes/Fort_08_Player_Interacts
 volumeMultiplier=1
}
