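# Sound cue for dropping the Shield Fish (Flopper) item, built on the Med Fish drop sound wave.
# The wave player feeds a Modulator with a fixed slight pitch raise, then an Enveloper that holds volume before quickly fading it out to trim the tail.
# Cue-level settings further raise pitch, lower volume, and apply weapon-drop attenuation and the loot-collect sound class.
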SoundCue /Game/Athena/Sounds/Items/Flopper/Shield_Fish_Drop_Cue
root=Enveloper_0
nodes [
 {
  id=Enveloper_0
  children [Modulator_0]
  properties {
   pitchCurve {editorCurveData {keys [{time=0}]}}
   volumeCurve {editorCurveData {keys [<at 0 1>, {arriveTangent=-1, interpMode=RCIM_Cubic, leaveTangent=-1, tangentMode=RCTM_Break, time=0.5, value=1}, <at 0.65 0>]}}
  }
 }
 {id=Modulator_0, children [WavePlayer_1], properties {pitchMax=1.02, pitchMin=1.02}}
 {id=WavePlayer_1, properties {soundWaveAssetPtr=&/Med_Fish_Drop}}
]
properties {
 attenuationSettings=&Fort_Weapon_Drop_Att
 maxDistance=3000
 pitchMultiplier=1.1
 soundClassObject=&Fort_Loot_Collect
 volumeMultiplier=0.45
}
