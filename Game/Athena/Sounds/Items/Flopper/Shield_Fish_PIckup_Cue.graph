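# Sound cue for picking up the Shield Fish, a Flopper item, in Athena.
# A single Hop_Flop_PU wave feeds a Modulator for random variation, then an Enveloper that fades volume to silence within a second.
# The cue plays pitched up and at reduced volume, using weapon-pickup attenuation and the loot-collect sound class.
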SoundCue /Game/Athena/Sounds/Items/Flopper/Shield_Fish_PIckup_Cue
root=Enveloper_0
nodes [
 {
  id=Enveloper_0
  children [Modulator_0]
  properties {
   pitchCurve {editorCurveData {keys [{time=0}]}}
   volumeCurve {editorCurveData {keys [<at 0 1>, {arriveTangent=-1, interpMode=RCIM_Cubic, leaveTangent=-1, tangentMode=RCTM_Break, time=0.4638, value=0.9363}, <at 0.75 0>]}}
  }
 }
 {id=Modulator_0, children [WavePlayer_0]}
 {id=WavePlayer_0, properties {soundWaveAssetPtr=&/Hop_Flop_PU}}
]
properties {
 attenuationSettings=&Fort_Weapon_Pickup_Att
 maxDistance=3500
 pitchMultiplier=1.15
 soundClassObject=&Fort_Loot_Collect
 volumeMultiplier=0.45
}
