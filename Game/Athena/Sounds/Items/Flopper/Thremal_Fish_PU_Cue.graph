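# Sound cue for picking up the thermal fish, one of the Flopper items in Athena.
# A single wave player feeds a Modulator left at its default settings, which is the root of the cue.
# It is routed to the loot-collect sound class with reduced volume and uses the shared weapon-drop attenuation for distance falloff.
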SoundCue /Game/Athena/Sounds/Items/Flopper/Thremal_Fish_PU_Cue
root=Modulator_0
nodes [
 {id=Modulator_0, children [WavePlayer_0]}
 {id=WavePlayer_0, properties {soundWaveAssetPtr=&/Thremal_Fish_PU}}
]
properties {
 attenuationSettings=&Fort_Weapon_Drop_Att
 maxDistance=3000
 soundClassObject=&Fort_Loot_Collect
 volumeMultiplier=0.45
}
