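# Sound cue played when the Zero Point Fish flopper item lands, judging by its name and wave assets.
# A mixer layers the generic flopper landing wave with a slightly louder rift fish landing wave, then feeds a modulator whose volume range is fixed at unity.
# It uses weapon drop attenuation and the loot collect sound class, at reduced overall volume.
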SoundCue /Game/Athena/Sounds/Items/Flopper/ZeroPoint_Fish_Landed_Cue
root=Modulator_1
nodes [
 {id=Mixer_1, children [WavePlayer_2,WavePlayer_4], properties {inputVolume [1,1.2]}}
 {id=Modulator_1, children [Mixer_1], properties {volumeMax=1, volumeMin=1}}
 {id=WavePlayer_2, properties {soundWaveAssetPtr=&/Flopper_Landed_01}}
 {id=WavePlayer_4, properties {soundWaveAssetPtr=&/Rift_Fish_Landed}}
]
properties {
 attenuationSettings=&Fort_Weapon_Drop_Att
 maxDistance=3000
 soundClassObject=&Fort_Loot_Collect
 volumeMultiplier=0.5
}
