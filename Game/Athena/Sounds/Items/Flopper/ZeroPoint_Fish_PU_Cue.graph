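# Sound cue for picking up the Zero Point fish, a Flopper item, routed to the loot collect sound class.
# A mixer layers the Rift fish pickup wave with a quieter Zero Point crystal consume wave.
# An enveloper pitches the crystal layer up by a constant amount and fades it out within about a second, reusing that consume sound as a short accent.
# Playback uses weapon drop attenuation.
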SoundCue /Game/Athena/Sounds/Items/Flopper/ZeroPoint_Fish_PU_Cue
root=Mixer_0
nodes [
 {
  id=Enveloper_0
  children [WavePlayer_0]
  properties {
   pitchCurve {editorCurveData {keys [<at -0.0001 1.4144>]}}
   volumeCurve {editorCurveData {keys [<at 0 1>,<at 0.3888 1.0024>,<at 0.8327 0.0456>,<at 1.0851 -0.0296>]}}
  }
 }
 {id=Mixer_0, children [WavePlayer_2,Enveloper_0], properties {inputVolume [1,0.65]}}
 {id=WavePlayer_0, properties {soundWaveAssetPtr=/Game/Sounds/Fort_GamePlay_Sounds/Player/ZeroPointCrystals/ZeroPointCrystal_Consume_01}}
 {id=WavePlayer_2, properties {soundWaveAssetPtr=&/Rift_Fish_PU}}
]
properties {
 attenuationSettings=&Fort_Weapon_Drop_Att
 maxDistance=3000
 soundClassObject=&Fort_Loot_Collect
 volumeMultiplier=0.65
}
